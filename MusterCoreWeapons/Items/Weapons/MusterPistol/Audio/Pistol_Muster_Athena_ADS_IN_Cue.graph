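# Sound cue for the Muster pistol's aim-down-sights-in foley.
# A Random node picks between two ADS-in waves with equal weight and feeds a Modulator at the root.
# It plays quietly in the weapon foley sound class with weapon reload attenuation, and a concurrency override limits it to one instance at a time.
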SoundCue /MusterCoreWeapons/Items/Weapons/MusterPistol/Audio/Pistol_Muster_Athena_ADS_IN_Cue
root=Modulator_0
nodes [
 {id=Modulator_0, children [Random_1]}
 {id=Random_1, children [WavePlayer_0,WavePlayer_3], properties {weights [1,1]}}
 {id=WavePlayer_0, properties {soundWaveAssetPtr=/MusterCoreWeapons/Items/Weapons/MusterPistol/Audio/Waves/Pistol_Muster_Athena_ADS_In_02}}
 {id=WavePlayer_3, properties {soundWaveAssetPtr=/MusterCoreWeapons/Items/Weapons/MusterPistol/Audio/Waves/Pistol_Muster_Athena_ADS_In_01}}
]
properties {
 attenuationSettings=&Fort_Weapon_Reload_Att
 bOverrideConcurrency=true
 maxDistance=2250
 soundClassObject=&Fort_WeaponFoley
 volumeMultiplier=0.1
 concurrencyOverrides {maxCount=1}
}
