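# Sound cue for the Muster pistol's aim-down-sights exit foley.
# A random node picks one of two equally weighted ADS-out waves and feeds it through a modulator.
# It plays very quietly in the weapon foley class using weapon reload attenuation, with a concurrency override limiting it to a single instance.
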SoundCue /MusterCoreWeapons/Items/Weapons/MusterPistol/Audio/Pistol_Muster_Athena_ADS_OUT_Cue
root=Modulator_0
nodes [
 {id=Modulator_0, children [Random_0]}
 {id=Random_0, children [WavePlayer_1,WavePlayer_2], properties {weights [1,1]}}
 {id=WavePlayer_1, properties {soundWaveAssetPtr=/MusterCoreWeapons/Items/Weapons/MusterPistol/Audio/Waves/Pistol_Muster_Athena_ADS_Out_01}}
 {id=WavePlayer_2, properties {soundWaveAssetPtr=/MusterCoreWeapons/Items/Weapons/MusterPistol/Audio/Waves/Pistol_Muster_Athena_ADS_Out_02}}
]
properties {
 attenuationSettings=&Fort_Weapon_Reload_Att
 bOverrideConcurrency=true
 maxDistance=2250
 soundClassObject=&Fort_WeaponFoley
 volumeMultiplier=0.03
 concurrencyOverrides {maxCount=1}
}
