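# Sound cue for the first-person firing sound of the Muster pistol (Athena variant).
# A random node picks one of six equally weighted fire waves and feeds a modulator left at default settings.
# It uses first-person weapon attenuation, the ranged-fire 1P sound class and a shared gunfire concurrency set, and also carries owner-limited, stop-oldest concurrency overrides.
# The debug flag is left on.
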SoundCue /MusterCoreWeapons/Items/Weapons/MusterPistol/Audio/Pistol_Muster_Athena_Fire_1P_Cue
root=Modulator_0
nodes [
 {id=Modulator_0, children [Random_1]}
 {
  id=Random_1
  children [WavePlayer_0, WavePlayer_5, WavePlayer_6, WavePlayer_7, WavePlayer_10, WavePlayer_11]
  properties {weights [1,1,1,1,1,1]}
 }
 {id=WavePlayer_0, properties {soundWaveAssetPtr=&/Pistol_Muster_Athena_Fire_1P_01}}
 {id=WavePlayer_10, properties {soundWaveAssetPtr=&/Pistol_Muster_Athena_Fire_1P_05}}
 {id=WavePlayer_11, properties {soundWaveAssetPtr=&/Pistol_Muster_Athena_Fire_1P_06}}
 {id=WavePlayer_5, properties {soundWaveAssetPtr=&/Pistol_Muster_Athena_Fire_1P_02}}
 {id=WavePlayer_6, properties {soundWaveAssetPtr=&/Pistol_Muster_Athena_Fire_1P_03}}
 {id=WavePlayer_7, properties {soundWaveAssetPtr=&/Pistol_Muster_Athena_Fire_1P_04}}
]
properties {
 attenuationSettings=&Fort_Weapon_Projectile_FirstPerson_Athena_Att
 bDebug=true
 maxDistance=2097152
 soundClassObject=&Fort_WeaponRanged_Fire_1P_PSM
 volumeMultiplier=0.7
 attenuationOverrides {manualReverbSendLevel=-1.0165, reverbSendMethod=Manual}
 concurrencyOverrides {bLimitToOwner=true, maxCount=3, resolutionRule=StopOldest, volumeScale=0.9}
 concurrencySet [&Weapon_Gunfire_FirstPerson_Conc]
}
