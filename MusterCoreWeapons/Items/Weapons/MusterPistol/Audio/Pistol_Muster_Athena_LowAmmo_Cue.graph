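# Sound cue for the low-ammo audio of the Muster pistol in MusterCoreWeapons.
# A short delay node at the root feeds an equally weighted random node that picks one of six low-ammo wave variations.
# It boosts volume, routes to the ranged-weapon low-ammo sound class, and overrides concurrency to prevent new instances once the limit is reached.
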SoundCue /MusterCoreWeapons/Items/Weapons/MusterPistol/Audio/Pistol_Muster_Athena_LowAmmo_Cue
root=Delay_0
nodes [
 {id=Delay_0, children [Random_0], properties {delayMin=0.05}}
 {id=Random_0, children [WavePlayer_0,WavePlayer_1,WavePlayer_2,WavePlayer_3,WavePlayer_4,WavePlayer_5], properties {weights [1,1,1,1,1,1]}}
 {id=WavePlayer_0, properties {soundWaveAssetPtr=&/Pistol_Muster_Athena_LowAmmo_01}}
 {id=WavePlayer_1, properties {soundWaveAssetPtr=&/Pistol_Muster_Athena_LowAmmo_02}}
 {id=WavePlayer_2, properties {soundWaveAssetPtr=&/Pistol_Muster_Athena_LowAmmo_03}}
 {id=WavePlayer_3, properties {soundWaveAssetPtr=&/Pistol_Muster_Athena_LowAmmo_04}}
 {id=WavePlayer_4, properties {soundWaveAssetPtr=&/Pistol_Muster_Athena_LowAmmo_05}}
 {id=WavePlayer_5, properties {soundWaveAssetPtr=&/Pistol_Muster_Athena_LowAmmo_06}}
]
properties {
 bOverrideConcurrency=true
 maxDistance=2097152
 soundClassObject=/Game/Sounds/FortSoundClassesAndModes/Fort_WeaponRanged_LowAmmo_Degraded_PSM
 volumeMultiplier=1.7
 concurrencyOverrides {resolutionRule=PreventNew}
}
