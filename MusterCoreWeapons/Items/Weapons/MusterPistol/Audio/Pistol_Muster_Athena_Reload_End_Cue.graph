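# Sound cue for the end of the Muster Pistol's reload.
# A Random node picks between two equally weighted reload-end wave variants.
# It plays in the weapon foley sound class with shared weapon reload attenuation, boosted volume and a concurrency override, and has its debug flag enabled.
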SoundCue /MusterCoreWeapons/Items/Weapons/MusterPistol/Audio/Pistol_Muster_Athena_Reload_End_Cue
root=Random_0
nodes [
 {id=Random_0, children [WavePlayer_0,WavePlayer_1], properties {weights [1,1]}}
 {id=WavePlayer_0, properties {soundWaveAssetPtr=&/Pistol_Muster_Athena_ReloadEnd_01}}
 {id=WavePlayer_1, properties {soundWaveAssetPtr=&/Pistol_Muster_Athena_ReloadEnd_02}}
]
properties {
 attenuationSettings=&Fort_Weapon_Reload_Att
 bDebug=true
 bOverrideConcurrency=true
 maxDistance=2250
 soundClassObject=&Fort_WeaponFoley
 volumeMultiplier=2
}
